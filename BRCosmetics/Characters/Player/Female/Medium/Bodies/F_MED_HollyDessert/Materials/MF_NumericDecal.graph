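# Material function that draws a one- or two-digit number decal from a number-sheet texture, for the HollyDessert character body.
# DigitTens and DigitOnes, read from custom primitive data, offset UVs into the NumberSheet strip, with each sample clipped by a procedural horizontal band mask.
# A nonzero tens digit or ForceTwoDigits shifts the layout and blends in the tens digit.
# Outputs are the composited digits and a combined mask of both bands.
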MaterialFunction /BRCosmetics/Characters/Player/Female/Medium/Bodies/F_MED_HollyDessert/Materials/MF_NumericDecal
graph {nodes [
  n1 = Constant() {r=0.045}
  n2 = ScalarParameter() {defaultValue=0.01, group=NumberDisplay, parameterName=H-Offset-Base}
  n3 = Saturate(None=Max(
    A=ScalarParameter() {group=NumberDisplay, parameterName=DigitTens, primitiveDataIndex=2}
    B=ScalarParameter() {group=NumberDisplay, parameterName=ForceTwoDigits}
   ))
  n4 = Add(
   A=FunctionInput(Preview=TextureCoordinate()) {bUsePreviewValueAsDefault=true, inputName=UVs, inputType=FunctionInput_Vector2}
   B=MaterialFunctionCall(X=LinearInterpolate(A=@n1, B=@n2, Alpha=@n3), Y=Constant()) {materialFunction=&MakeFloat2}.Result
  )
  n5 = Multiply(A=ComponentMask(None=@n4) {r=true}, B=Constant() {r=-100})
  n6 = Constant() {r=53.85}
  n7 = Add(A=Abs(None=Add(A=@n5, B=@n6)), B=Constant() {r=-4.5})
  n8 = Constant() {r=15}
  n9 = OneMinus(None=Saturate(None=Multiply(A=@n7, B=@n8)))
  n10 = ScalarParameter() {group=NumberDisplay, parameterName=DigitOnes, primitiveDataIndex=3}
  n11 = Add(A=Divide(A=ScalarParameter() {defaultValue=10, group=NumberDisplay, parameterName=NumberOFDigits})) {desc=TODO: Need a better way to offset this}
  n12 = Add(
   A=Divide(A=Subtract(A=@n10, B=@n11), B=ScalarParameter() {defaultValue=10, group=NumberDisplay, parameterName=NumberOFDigits})
   B=Divide(A=Constant() {r=1}, B=ScalarParameter() {defaultValue=10, group=NumberDisplay, parameterName=NumberOFDigits})
  )
  n13 = Constant()
  n14 = TextureSample(
   UVs=Add(A=@n4, B=MaterialFunctionCall(X=@n12, Y=@n13) {materialFunction=&MakeFloat2}.Result)
   Tex=NamedRerouteUsage()
  ) {
   samplerSource=SSM_Clamp_WorldGroupSettings
   samplerType=SAMPLERTYPE_LinearColor
   texture=/BRCosmetics/Accessories/FORT_Backpacks/Backpack_M_MED_AgileRug/Textures/T_NumberSet
  }
  n15 = Multiply(A=@n9, B=@n14.RGBA)
  n16 = Constant() {r=15}
  n17 = Constant() {r=47}
  n18 = Add(A=Constant() {r=-4.5}, B=Abs(None=Add(A=@n17, B=@n5)))
  n19 = OneMinus(None=Saturate(None=Multiply(A=@n16, B=@n18)))
  n20 = ScalarParameter() {group=NumberDisplay, parameterName=DigitTens, primitiveDataIndex=2}
  n21 = Divide(A=ScalarParameter() {defaultValue=10, group=NumberDisplay, parameterName=NumberOFDigits})
  n22 = Add(
   A=Divide(A=Subtract(A=@n20, B=@n21), B=ScalarParameter() {defaultValue=10, group=NumberDisplay, parameterName=NumberOFDigits})
   B=Divide(A=Constant() {r=1}, B=ScalarParameter() {defaultValue=10, group=NumberDisplay, parameterName=NumberOFDigits})
  )
  n23 = Constant()
  n24 = Constant2Vector() {r=-0.03}
  n25 = TextureSample(
   UVs=Add(
    A=MaterialFunctionCall(X=@n22, Y=@n23) {materialFunction=&MakeFloat2}.Result
    B=Add(A=@n4, B=@n24)
   )
   Tex=NamedRerouteUsage()
  ) {
   samplerSource=SSM_Clamp_WorldGroupSettings
   samplerType=SAMPLERTYPE_LinearColor
   texture=&T_NumberSet
  }
  n26 = Multiply(A=@n19, B=@n25.RGBA)
  n27 = Max(
   A=ScalarParameter() {group=NumberDisplay, parameterName=DigitTens, primitiveDataIndex=2}
   B=ScalarParameter() {defaultValue=1, group=NumberDisplay, parameterName=ForceTwoDigits}
  )
  FunctionOutput(None=LinearInterpolate(A=@n15, B=Max(A=@n26, B=@n15), Alpha=Saturate(None=@n27)))
  FunctionOutput(None=Max(A=@n19, B=@n9)) {outputName=0-1 Mask, sortPriority=2}
  n28 = TextureObjectParameter() {
   group=NumberDisplay
   parameterName=NumberSheet
   samplerType=SAMPLERTYPE_Masks
   texture=/BRCosmetics/Characters/Player/Female/Medium/Bodies/F_MED_HollyDessert/Textures/T_NumberSet
  }
  NamedRerouteDeclaration(None=@n28) {name=NumberTexture, nodeColor=<rgba 0.5234 1 0 1>}
  Saturate()
 ]}
