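# Base cloth-shading-model material for the HollyDessert female body cosmetic, adding a colored numeric decal to the 2023 anime character shader.
# The shared Pre, Calculations and Post functions form the shading chain, with named reroutes carrying lit and shaded colors.
# Offset, scaled and saturated UVs drive MF_NumericDecal, whose red channel blends NumberColor into both lit and shaded colors.
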Material /BRCosmetics/Characters/Player/Female/Medium/Bodies/F_MED_HollyDessert/Materials/M_HollyDessert_Body_Base
graph {
 properties {
  bAutomaticallySetUsageInEditor=false
  bUseMaterialAttributes=true
  bUsedWithMeshDeformer=true
  bUsedWithMorphTargets=true
  bUsedWithSkeletalMesh=true
  bUsedWithStaticLighting=true
  pixelDepthOffsetMode=PDOM_Legacy
  shadingModel=MSM_Cloth
  usedShadingModels=Cloth
 }
 nodes [
  n1 = MaterialFunctionCall!(CustomUV) {materialFunction=/Game/Global/Materials/Character_Parent_2023/Anime_Parent_2023/MaterialFunctions/MFC_AnimeShader_2023_Pre}
  n2 = MaterialFunctionCall!(CustomLightVector, OverrideNormal_WS, STT=@n1.STT, Normal_Map=@n1.STT, DFL=@n1.STT, FX=@n1.STT) {materialFunction=/Game/Global/Materials/Character_Parent_2023/Anime_Parent_2023/MaterialFunctions/MFC_AnimeShader_2023_Calculations}
  MaterialFunctionCall!(
   Shading_Gradient=@n2.Shading_Gradient
   Threshold_Gradient=@n2.Shading_Gradient
   Specular_Gradient=@n2.Shading_Gradient
   Sheen_Mask=@n2.Shading_Gradient
   Thickness=@n2.Shading_Gradient
   DFL_Mask=@n2.Shading_Gradient
   FX=@n2.Shading_Gradient
   Metallic_Gradient=@n2.Shading_Gradient
   RimLight_Gradient=@n2.Shading_Gradient
   NormalBuffer=@n2.Shading_Gradient
   --------
   CINE_LightSources=@n2.Shading_Gradient
   CINE_LightColor=@n2.Shading_Gradient
   ---------
   Color_Lit=NamedRerouteUsage()
   Color_Shaded=NamedRerouteUsage()
   SRM=@n1.SRM
   Emissive=@n1.SRM
   DFL_Color=@n1.SRM
   Specular_Color=@n1.SRM
   RimLight_Color=@n1.SRM
  ) {materialFunction=/Game/Global/Materials/Character_Parent_2023/Anime_Parent_2023/MaterialFunctions/MFC_AnimeShader_2023_Post}
  NamedRerouteDeclaration(None=@n1.Color_Lit) {name=IN_Color_Lit, nodeColor=<rgba 0 1 0.3906 1>}
  NamedRerouteDeclaration(None=@n1.Color_Shaded) {name=IN_Color_Shaded, nodeColor=<rgba 0 0.2693 1 1>}
  n3 = VectorParameter() {group=HollyDessert, parameterName=NumberColor, defaultValue=<rgba 0 0 0 1>}
  n4 = VectorParameter() {group=HollyDessert, parameterName=NumberUVs_Offset, defaultValue=<rgba 0 0 0 1>}
  n5 = VectorParameter() {group=HollyDessert, parameterName=NumberUVs_Scale, defaultValue=<rgba 1 1 1 1>}
  n6 = Multiply(A=Add(A=TextureCoordinate(), B=AppendVector(A=@n4.R, B=@n4.R)) {constB=0}, B=AppendVector(A=@n5.R, B=@n5.R))
  n7 = ComponentMask(None=MaterialFunctionCall!(UVs=Saturate(None=@n6)) {materialFunction=&/MF_NumericDecal}.Result) {r=true}
  NamedRerouteDeclaration(None=LinearInterpolate(A=NamedRerouteUsage(), B=@n3.RGB, Alpha=@n7)) {name=OUT_Color_Lit, nodeColor=<rgba 0.1719 1 0 1>}
  NamedRerouteDeclaration(None=LinearInterpolate(A=NamedRerouteUsage(), B=@n3.RGB, Alpha=@n7)) {name=OUT_Color_Shaded, nodeColor=<rgba 0.5469 0 1 1>}
 ]
}
