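# Physics asset for the HollyDessert female medium cosmetic's face accessory mesh, set up for dynamic hair motion.
# Five hair chains (center, front left/right, side left/right) each link a small base capsule on the head to three progressively thicker capsules.
# Chain swing limits widen toward the tips with very tight twist, while base, head and neck joints are free.
# Head, neck and upper spine capsules provide the body shapes.
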PhysicsAsset /BRCosmetics/Characters/Player/Female/Medium/Bodies/F_MED_HollyDessert/Meshes/Parts/F_MED_HollyDessert_FaceAcc_Physics
bodies {
 base_hair_c [{shapeType=Capsule, length=0.2, radius=0.5, shapeName=base_hair_c_capsule, rotation=<rot 90 0 0>}]
 base_hair_fr_l [{shapeType=Capsule, length=0.2, radius=0.5, shapeName=base_hair_fr_l_capsule, rotation=<rot 90 0 0>}]
 base_hair_fr_r [{shapeType=Capsule, length=0.2, radius=0.5, shapeName=base_hair_fr_r_capsule, rotation=<rot 90 0 0>}]
 base_hair_side_l [{shapeType=Capsule, length=0.2, radius=0.5, shapeName=base_hair_side_l_capsule, rotation=<rot 90 0 0>}]
 base_hair_side_r [{shapeType=Capsule, length=0.2, radius=0.5, shapeName=base_hair_side_r_capsule, rotation=<rot 90 0 0>}]
 dyn_hair_01_c [{shapeType=Capsule, length=2, radius=1.2749, shapeName=dyn_hair_01_c_capsule, center=<2 0.1567 0>, rotation=<rot 90 0 180>}]
 dyn_hair_02_c [{shapeType=Capsule, length=2.2257, radius=1.3995, shapeName=dyn_hair_02_c_capsule, center=<2.6218 0.5023 0>, rotation=<rot 90 180 -3.2311>}]
 dyn_hair_03_c [{shapeType=Capsule, length=5.754, radius=1.9043, shapeName=dyn_hair_03_c_capsule, center=<5.4367 2.6831 -0.002>, rotation=<rot 0 90 180>}]
 dyn_hair_fr_01_l [{shapeType=Capsule, length=2, radius=1.2749, shapeName=dyn_hair_fr_01_l_capsule, center=<1.878 0.2256 1.5508>, rotation=<rot 90 0 180>}]
 dyn_hair_fr_01_r [{shapeType=Capsule, length=2, radius=1.2749, shapeName=dyn_hair_fr_01_r_capsule, center=<1.878 0.2256 -1.5508>, rotation=<rot 90 0 180>}]
 dyn_hair_fr_02_l [{
   shapeType=Capsule
   length=2.7423
   radius=1.3995
   shapeName=dyn_hair_fr_02_l_capsule
   center=<2.5614 0.5259 1.4184>
   rotation=<rot 90 180 -6.6886>
  }]
 dyn_hair_fr_02_r [{
   shapeType=Capsule
   length=2.7423
   radius=1.3995
   shapeName=dyn_hair_fr_02_r_capsule
   center=<2.5614 0.5259 -1.4184>
   rotation=<rot 90 178.2101 -8.4784>
  }]
 dyn_hair_fr_03_l [{
   shapeType=Capsule
   length=2.6186
   radius=1.9043
   shapeName=dyn_hair_fr_03_r_capsule
   center=<4.4834 0.2541 0.2066>
   rotation=<rot 10.7601 90 -179.9999>
  }]
 dyn_hair_fr_03_r [{
   shapeType=Capsule
   length=2.6186
   radius=1.9043
   shapeName=dyn_hair_fr_03_r_capsule
   center=<4.4834 0.2541 0.2066>
   rotation=<rot -12.8837 90 179.9999>
  }]
 dyn_hair_side_01_l [{shapeType=Capsule, length=2, radius=1.2749, shapeName=dyn_hair_side_01_l_capsule, center=<2 0.3605 0>, rotation=<rot 90 0 180>}]
 dyn_hair_side_01_r [{shapeType=Capsule, length=2, radius=1.2749, shapeName=dyn_hair_side_01_r_capsule, center=<2 0.3605 0>, rotation=<rot 90 0 180>}]
 dyn_hair_side_02_l [{shapeType=Capsule, length=2.2257, radius=1.3995, shapeName=dyn_hair_side_02_l_capsule, center=<2.6228 0.7074 0>, rotation=<rot 90 0 180>}]
 dyn_hair_side_02_r [{shapeType=Capsule, length=2.2257, radius=1.3995, shapeName=dyn_hair_side_02_r_capsule, center=<2.6228 0.7074 0>, rotation=<rot 90 0 180>}]
 dyn_hair_side_03_l [{
   shapeType=Capsule
   length=1.478
   radius=1.9043
   shapeName=dyn_hair_side_03_l_capsule
   center=<5.4367 0.3342 -0.0029>
   rotation=<rot 0 90 180>
  }]
 dyn_hair_side_03_r [{shapeType=Capsule, length=1.478, radius=1.9043, shapeName=dyn_hair_side_03_r_capsule, center=<5.4367 0.3342 0.0029>, rotation=<rot 0 90 180>}]
 head [
  {shapeType=Capsule, length=5.7507, radius=4.6541, center=<1.922 0.1354 0>, rotation=<rot -90 153.4349 116.5651>}
  {shapeType=Capsule, length=5.7507, radius=4.6541, center=<5.6228 0.2136 0>, rotation=<rot -90 153.4349 116.5651>}
  {shapeType=Capsule, length=4.2982, radius=6.0317, shapeName=head_capsule, center=<2.103 4.6 0>, rotation=<rot 90 0 0>}
 ]
 neck_01 [{shapeType=Capsule, length=5.3478, radius=3.9297, shapeName=neck_01_capsule, center=<1.2408 0.3914 0>, rotation=<rot 90 180 -177.3684>}]
 spine_05 [{shapeType=Capsule, length=9.678, radius=7.7732, center=<3.902 -0.4622 0>, rotation=<rot 0 12.9661 0>}]
}
constraints [
 {
  bone1Name=neck_01
  bone2Name=spine_05
  swing1LimitDegrees=45
  swing1Motion=Free
  swing2LimitDegrees=45
  swing2Motion=Free
  twistLimitDegrees=45
  twistMotion=Free
 }
 {
  bone1Name=base_hair_c
  bone2Name=head
  swing1LimitDegrees=45
  swing1Motion=Free
  swing2LimitDegrees=45
  swing2Motion=Free
  twistLimitDegrees=45
  twistMotion=Free
 }
 {bone1Name=dyn_hair_01_c, bone2Name=base_hair_c, swing1LimitDegrees=5, swing2LimitDegrees=5, twistLimitDegrees=1}
 {bone1Name=dyn_hair_02_c, bone2Name=dyn_hair_01_c, swing1LimitDegrees=7, swing2LimitDegrees=7, twistLimitDegrees=1}
 {bone1Name=dyn_hair_03_c, bone2Name=dyn_hair_02_c, swing1LimitDegrees=10, swing2LimitDegrees=10, twistLimitDegrees=1}
 {
  bone1Name=base_hair_fr_l
  bone2Name=head
  swing1LimitDegrees=45
  swing1Motion=Free
  swing2LimitDegrees=45
  swing2Motion=Free
  twistLimitDegrees=45
  twistMotion=Free
 }
 {bone1Name=dyn_hair_fr_01_l, bone2Name=base_hair_fr_l, swing1LimitDegrees=5, swing2LimitDegrees=5, twistLimitDegrees=1}
 {bone1Name=dyn_hair_fr_02_l, bone2Name=dyn_hair_fr_01_l, swing1LimitDegrees=7, swing2LimitDegrees=7, twistLimitDegrees=1}
 {bone1Name=dyn_hair_fr_03_l, bone2Name=dyn_hair_fr_02_l, swing1LimitDegrees=10, swing2LimitDegrees=10, twistLimitDegrees=1}
 {
  bone1Name=base_hair_fr_r
  bone2Name=head
  swing1LimitDegrees=45
  swing1Motion=Free
  swing2LimitDegrees=45
  swing2Motion=Free
  twistLimitDegrees=45
  twistMotion=Free
 }
 {bone1Name=dyn_hair_fr_01_r, bone2Name=base_hair_fr_r, swing1LimitDegrees=5, swing2LimitDegrees=5, twistLimitDegrees=1}
 {bone1Name=dyn_hair_fr_02_r, bone2Name=dyn_hair_fr_01_r, swing1LimitDegrees=7, swing2LimitDegrees=7, twistLimitDegrees=1}
 {bone1Name=dyn_hair_fr_03_r, bone2Name=dyn_hair_fr_02_r, swing1LimitDegrees=10, swing2LimitDegrees=10, twistLimitDegrees=1}
 {
  bone1Name=base_hair_side_l
  bone2Name=head
  swing1LimitDegrees=45
  swing1Motion=Free
  swing2LimitDegrees=45
  swing2Motion=Free
  twistLimitDegrees=45
  twistMotion=Free
 }
 {bone1Name=dyn_hair_side_01_l, bone2Name=base_hair_side_l, swing1LimitDegrees=5, swing2LimitDegrees=5, twistLimitDegrees=1}
 {bone1Name=dyn_hair_side_02_l, bone2Name=dyn_hair_side_01_l, swing1LimitDegrees=7, swing2LimitDegrees=7, twistLimitDegrees=1}
 {bone1Name=dyn_hair_side_03_l, bone2Name=dyn_hair_side_02_l, swing1LimitDegrees=10, swing2LimitDegrees=10, twistLimitDegrees=1}
 {
  bone1Name=base_hair_side_r
  bone2Name=head
  swing1LimitDegrees=45
  swing1Motion=Free
  swing2LimitDegrees=45
  swing2Motion=Free
  twistLimitDegrees=45
  twistMotion=Free
 }
 {bone1Name=dyn_hair_side_01_r, bone2Name=base_hair_side_r, swing1LimitDegrees=5, swing2LimitDegrees=5, twistLimitDegrees=1}
 {bone1Name=dyn_hair_side_02_r, bone2Name=dyn_hair_side_01_r, swing1LimitDegrees=7, swing2LimitDegrees=7, twistLimitDegrees=1}
 {bone1Name=dyn_hair_side_03_r, bone2Name=dyn_hair_side_02_r, swing1LimitDegrees=10, swing2LimitDegrees=10, twistLimitDegrees=1}
 {
  bone1Name=head
  bone2Name=neck_01
  swing1LimitDegrees=45
  swing1Motion=Free
  swing2LimitDegrees=45
  swing2Motion=Free
  twistLimitDegrees=45
  twistMotion=Free
 }
]
