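# Animation Blueprint for the HollyDessert female face accessory, adding physics-driven secondary motion to a pose copied from the attached parent mesh.
# Two near-identical RigidBody nodes simulate in head-bone space with damped component motion and overridden gravity.
# A PickaxeSwing flag, toggled by swing begin/end events, blends to the lower-alpha simulation during swings.
# Each update also maps spine_05-to-head yaw into hair rotation and translation variables, which the AnimGraph never reads.
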AnimBlueprint /BRCosmetics/Characters/Player/Female/Medium/Bodies/F_MED_HollyDessert/Meshes/Parts/F_MED_HollyDessert_FaceAcc_AnimBP
target_skeleton=&/F_MED_HollyDessert_FaceAcc_Skeleton
graphs [
 AnimGraph {nodes [
   n1 = RigidBody!(
    ComponentPose=Local To Component(LocalPose=Copy Pose From Mesh() {bUseAttachedParent=true})
    Alpha=0.5
    OverrideWorldGravity:Vector="0,0,-980"
    ExternalForce
   ) -> (Pose) {
    bOverrideWorldGravity=true
    overridePhysicsAsset=&/F_MED_HollyDessert_FaceAcc_Physics
    simulationSpace=BaseBoneSpace
    alphaBoolBlend {blendOption=Linear}
    alphaScaleBiasClamp {bInterpResult=true}
    baseBoneRef {boneName=head}
    componentAppliedLinearAccClamp=<3000 3000 3000>
    componentLinearAccScale=<0.3 0.3 0.3>
    componentLinearVelScale=<0.3 0.3 0.3>
    simSpaceSettings {
     maxAngularAcceleration=1000
     maxAngularVelocity=3
     maxLinearAcceleration=700
     maxLinearVelocity=500
     velocityScaleZ=0
     worldAlpha=0.7
     externalLinearDragV=<0.2 0.2>
    }
   }
   n2 = RigidBody(ComponentPose=Local To Component(LocalPose=Copy Pose From Mesh() {bUseAttachedParent=true}), Alpha=0.7, OverrideWorldGravity="0,0,-980") {
    bOverrideWorldGravity=true
    overridePhysicsAsset=&F_MED_HollyDessert_FaceAcc_Physics
    simulationSpace=BaseBoneSpace
    alphaBoolBlend {blendOption=Linear}
    alphaScaleBiasClamp {bInterpResult=true}
    baseBoneRef {boneName=head}
    componentAppliedLinearAccClamp=<3000 3000 3000>
    componentLinearAccScale=<0.3 0.3 0.3>
    componentLinearVelScale=<0.3 0.3 0.3>
    simSpaceSettings {
     maxAngularAcceleration=1000
     maxAngularVelocity=3
     maxLinearAcceleration=700
     maxLinearVelocity=500
     velocityScaleZ=0
     worldAlpha=0.7
     externalLinearDragV=<0.2 0.2>
    }
   }
   Output Pose(Result=Blend Poses by bool(
     BlendPose_0=Component To Local(ComponentPose=@n1)
     BlendPose_1=Component To Local(ComponentPose=@n2)
     bActiveValue=Get PickaxeSwing() -> (PickaxeSwing:bool)
    ) {blendType=Linear}) {layerGroup=None}
  ]}
 EventGraph {nodes [
   Event BlueprintUpdateAnimation()
   n1 = GetOwnerSkeletalMeshComponent()
   n2 = BreakRotator(InRot=BreakTransform(InTransform=GetDeltaTransformFromRefPose(self=@n1, BoneName=spine_05, BaseName=head)).Rotation)
   > Set hair_rotate(hair_rotate:Rotator=MakeRotator(Roll=MapRangeClamped(Value=@n2.Yaw, InRangeA=0, InRangeB=-30, OutRangeB=35))) -> (Output_Get:Rotator)
   > Set hair_translation(hair_translation:Vector=MakeVector(Y=MapRangeClamped(Value=@n2.Yaw, InRangeA=0, InRangeB=-30, OutRangeB=-10))) -> (Output_Get:Vector)
   Event OnBeginPickaxeSwing() {is=title K2Node_Event}
   > Set PickaxeSwing(PickaxeSwing:bool=1) -> (Output_Get:bool)
   Event OnEndPickaxeSwing() {is=title K2Node_Event}
   > Set PickaxeSwing(PickaxeSwing=false)
  ]}
]
variables [
 hair_rotate {type=Rotator}
 hair_translation {type=Vector}
 PickaxeSwing {type=boolean}
]
